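# Niagara one-shot glow effect, named as a fog-free spherical-density sphere for a geode, in the FX reuse examples.
# A single Glow emitter burst-spawns a short-lived, warm orange mesh particle, scaled up and offset upward.
# Dynamic material parameters set at spawn feed the mesh material, presumably shaping the density look.
# A curve-driven alpha scale flashes in sharply, then eases out over the particle's life, with no user parameters exposed.
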NiagaraSystem /Game/NeverCook/FX/Niagara_FX_Reuse_Examples/Glow/NS_Geode_SphericalDensity_Sphere_NoFog
dependencies {usedDataInterfaces [&NiagaraDataInterfaceCurve], usedDynamicInputs [], usedModules [&SystemState]}
emitters {Glow {
  emitterUpdateScript [
   &EmitterState()
   &SpawnBurst_Instantaneous(Use Spawn Probability=false, UseLoopCountLimit=false)
  ]
  particleSpawnScript [
   &InitializeParticle(
    Color=<rgba 0.8228 0.4948 0.1422 1>
    Lifetime=0.5
    Mesh Scale Mode=::NewEnumerator3
    Mesh Uniform Scale=2
    Position Offset=<0 0 30>
    UsePositionOffset=true
    Write Material Random=true
   )
   &DynamicMaterialParameters(
    Index 0 Param 1=1.5
    Index 0 Param 2=1
    Index 0 Param 3=5
    Index 0 Param 4=1
    Module.Index0Param0Write=true
    Module.Index0Param1Write=true
    Module.Index0Param2Write=true
    Module.Index0Param3Write=true
   )
  ]
  particleUpdateScript [
   &ParticleState()
   &ScaleColor(
    Scale Alpha=&FloatFromCurve(
     FloatCurve=DataInterface() {propertyValues {Curve {keys [
         {arriveTangent=64.8538, interpMode=RCIM_Cubic, leaveTangent=64.8538, tangentMode=RCTM_User, time=0, value=0}
         {arriveTangent=-3.4473, interpMode=RCIM_Cubic, leaveTangent=-3.4473, tangentMode=RCTM_User, time=0.05, value=1}
         {interpMode=RCIM_Cubic, tangentMode=RCTM_User, time=1, value=0}
        ]}}}
     Scale Curve=0.2
    )
    Scale RGB=&VectorFromFloat(Value=1)
    ScaleA=true
    ScaleRGB=true
   )
  ]
  rendererClasses [&NiagaraMeshRendererProperties]
  renderers [{rendererClass=&NiagaraMeshRendererProperties, rendererIndex=0}]
 }}
summary {
 systemName=NS_Geode_SphericalDensity_Sphere_NoFog
 emitters [{emitterName=Glow, rendererClasses [&NiagaraMeshRendererProperties]}]
 userVariables []
}
